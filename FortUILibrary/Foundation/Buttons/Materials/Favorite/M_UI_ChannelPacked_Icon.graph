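# Translucent UI material, filed under Favorite buttons, that renders an icon from up to four shapes packed into one texture's channels.
# Center-scaled, offset UVs sample the texture, whose channels optionally pass through signed-distance-field smoothstep or outline shaping and per-channel inversion.
# Channels are tinted individually, layered alpha to red on top, or share one color.
# Static switches toggle alpha-channel use and max-combined versus layered opacity, scaled by global alpha.
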Material /FortUILibrary/Foundation/Buttons/Materials/Favorite/M_UI_ChannelPacked_Icon
graph {
 properties {
  bUsedWithStaticMesh=false
  blendMode=BLEND_Translucent
  materialDomain=MD_UI
  pixelDepthOffsetMode=PDOM_Legacy
  usedShadingModels=Default Lit
 }
 nodes [
  n1 = TextureSample(UVs=NamedRerouteUsage(), Tex=TextureObjectParameter() {group=0 - Global, parameterName=Texture, sortPriority=0}) {samplerSource=SSM_Clamp_WorldGroupSettings}
  n2 = Constant()
  NamedRerouteDeclaration(None=AppendVector(A=@n1.RGB, B=StaticSwitchParameter(True=@n1.A, False=@n2) {group=0 - Global, parameterName=Use Alpha Channel, sortPriority=0})) {name=Texture Channels, nodeColor=<rgba 0 0 0 1>}
  n3 = Max(
   A=Multiply(
    A=NamedRerouteUsage()
    B=ScalarParameter() {defaultValue=1, group=4 - Opacity, parameterName=Red Channel Opacity, sliderMax=1, sortPriority=5}
   )
   B=Multiply(
    A=NamedRerouteUsage()
    B=ScalarParameter() {defaultValue=1, group=4 - Opacity, parameterName=Green Channel Opacity, sliderMax=1, sortPriority=6}
   )
  )
  n4 = Multiply(
   A=NamedRerouteUsage()
   B=ScalarParameter() {defaultValue=1, group=4 - Opacity, parameterName=Blue Channel Opacity, sliderMax=1, sortPriority=7}
  )
  n5 = Multiply(
   A=NamedRerouteUsage()
   B=ScalarParameter() {defaultValue=1, group=4 - Opacity, parameterName=Alpha Channel Opacity, sliderMax=1, sortPriority=8}
  )
  n6 = Constant()
  NamedRerouteDeclaration(None=Max(A=Max(A=@n3, B=@n4), B=StaticSwitchParameter(True=@n5, False=@n6) {group=0 - Global, parameterName=Use Alpha Channel, sortPriority=0})) {name=Combined Alpha, nodeColor=<rgba 0 0 0 1>}
  n7 = ScalarParameter() {group=2 - Distance Field, parameterName=Min, sliderMax=1, sliderMin=-1, sortPriority=31}
  n8 = ScalarParameter() {defaultValue=0.01, group=2 - Distance Field, parameterName=Max, sliderMax=1, sliderMin=-1}
  n9 = ConstantBiasScale(None=NamedRerouteUsage()) {bias=-0.5, scale=2}
  n10 = StaticSwitchParameter(True=OneMinus(None=Abs(None=@n9)), False=@n9) {group=2 - Distance Field, parameterName=Display as Outline, sortPriority=1}
  n11 = StaticSwitchParameter(True=Saturate(None=SmoothStep(Min=@n7, Max=@n8, Value=@n10)), False=NamedRerouteUsage()) {defaultValue=true, group=2 - Distance Field, parameterName=Use Signed Distance Field, sortPriority=0}
  n12 = ComponentMask(None=@n11) {r=true}
  n13 = ComponentMask(None=@n11) {g=true}
  n14 = ComponentMask(None=@n11) {b=true}
  n15 = ComponentMask(None=@n11) {a=true}
  n16 = MaterialFunctionCall(
   X=StaticSwitchParameter(True=OneMinus(None=@n12), False=@n12) {group=0 - Global, parameterName=Inverse Red Channel, sortPriority=1}
   Y=StaticSwitchParameter(True=OneMinus(None=@n13), False=@n13) {group=0 - Global, parameterName=Inverse Green Channel, sortPriority=2}
   Z=StaticSwitchParameter(True=OneMinus(None=@n14), False=@n14) {group=0 - Global, parameterName=Inverse Blue Channel, sortPriority=3}
   A=StaticSwitchParameter(True=OneMinus(None=@n15), False=@n15) {group=0 - Global, parameterName=Inverse Alpha Channel, sortPriority=4}
  ) {materialFunction=&MakeFloat4}
  n17 = MaterialFunctionCall(Float4=NamedRerouteDeclaration(None=Saturate(None=@n16.Result)) {name=Distance Field Pass, nodeColor=<rgba 0 0 0 1>}) {materialFunction=&BreakOutFloat4Components}
  NamedRerouteDeclaration(None=@n17.A) {name=Alpha Channel, nodeColor=<rgba 0 0 0 1>}
  n18 = LinearInterpolate(
   B=ScalarParameter() {defaultValue=1, group=4 - Opacity, parameterName=Alpha Channel Opacity, sliderMax=1, sortPriority=8}
   Alpha=NamedRerouteUsage()
  )
  n19 = Constant()
  n20 = LinearInterpolate(
   A=LinearInterpolate(
    A=StaticSwitchParameter(True=@n18, False=@n19) {group=0 - Global, parameterName=Use Alpha Channel, sortPriority=0}
    B=ScalarParameter() {defaultValue=1, group=4 - Opacity, parameterName=Blue Channel Opacity, sliderMax=1, sortPriority=7}
    Alpha=NamedRerouteUsage()
   )
   B=ScalarParameter() {defaultValue=1, group=4 - Opacity, parameterName=Green Channel Opacity, sliderMax=1, sortPriority=6}
   Alpha=NamedRerouteUsage()
  )
  n21 = ScalarParameter() {defaultValue=1, group=4 - Opacity, parameterName=Red Channel Opacity, sliderMax=1, sortPriority=5}
  n22 = NamedRerouteUsage()
  NamedRerouteDeclaration(None=Saturate(None=LinearInterpolate(A=@n20, B=@n21, Alpha=@n22) {constAlpha=1})) {name=Layered Alpha, nodeColor=<rgba 0 0 0 1>}
  n23 = NamedRerouteUsage()
  n24 = NamedRerouteUsage()
  NamedRerouteDeclaration(None=Multiply(
    A=StaticSwitchParameter(True=@n23, False=@n24) {group=4 - Opacity, parameterName=Use Combined Transparency, sortPriority=1}
    B=ScalarParameter() {defaultValue=1, group=4 - Opacity, parameterName=Global Alpha, sliderMax=1, sortPriority=0}
   )) {name=Opacity, nodeColor=<rgba 0 0 0 1>}
  n25 = VectorParameter() {group=3 - Color, parameterName=Alpha Channel Color, sortPriority=35, defaultValue=<rgba 0 0.7682 0.7495 1>}
  n26 = VectorParameter() {group=3 - Color, parameterName=Blue Channel Color, sortPriority=34, defaultValue=<rgba 0 0.0744 0.7682 1>}
  n27 = NamedRerouteUsage()
  n28 = Multiply(A=ScalarParameter() {defaultValue=1, group=4 - Opacity, parameterName=Blue Channel Opacity, sliderMax=1, sortPriority=7}, B=@n27)
  n29 = StaticSwitchParameter(
   True=LinearInterpolate(
    A=@n25.RGB
    B=@n26.RGB
    Alpha=StaticSwitchParameter(True=@n28, False=@n27) {group=4 - Opacity, parameterName=Use Combined Transparency, sortPriority=1}
   )
   False=@n26.RGB
  ) {group=0 - Global, parameterName=Use Alpha Channel, sortPriority=0}
  n30 = VectorParameter() {group=3 - Color, parameterName=Green Channel Color, sortPriority=33, defaultValue=<rgba 1 0.8111 0 1>}
  n31 = NamedRerouteUsage()
  n32 = StaticSwitchParameter(
   True=Multiply(A=ScalarParameter() {defaultValue=1, group=4 - Opacity, parameterName=Green Channel Opacity, sliderMax=1, sortPriority=6}, B=@n31)
   False=@n31
  ) {group=4 - Opacity, parameterName=Use Combined Transparency, sortPriority=1}
  n33 = NamedRerouteUsage()
  n34 = Multiply(A=ScalarParameter() {defaultValue=1, group=4 - Opacity, parameterName=Red Channel Opacity, sliderMax=1, sortPriority=5}, B=@n33)
  NamedRerouteDeclaration(None=LinearInterpolate(
    A=LinearInterpolate(A=@n29, B=@n30.RGB, Alpha=@n32)
    B=VectorParameter() {group=3 - Color, parameterName=Red Channel Color, defaultValue=<rgba 1 0.0411 0 1>}.RGB
    Alpha=StaticSwitchParameter(True=@n34, False=@n33) {group=4 - Opacity, parameterName=Use Combined Transparency, sortPriority=1}
   )) {name=Separated Colors, nodeColor=<rgba 0 0 0 1>}
  NamedRerouteDeclaration(None=@n17.R) {name=Red Channel, nodeColor=<rgba 0 0 0 1>}
  NamedRerouteDeclaration(None=@n17.G) {name=Green Channel, nodeColor=<rgba 0 0 0 1>}
  NamedRerouteDeclaration(None=@n17.B) {name=Blue Channel, nodeColor=<rgba 0 0 0 1>}
  n35 = ScalarParameter() {defaultValue=1, group=1 - Transform, parameterName=X Scale, sortPriority=0}
  n36 = ScalarParameter() {defaultValue=1, group=1 - Transform, parameterName=Y Scale, sortPriority=0}
  n37 = Divide(
   A=Subtract(A=MaterialFunctionCall() {materialFunction=&GetUserInterfaceUV}.Normalized UV) {constB=0.5}
   B=StaticSwitchParameter(True=ScalarParameter() {defaultValue=1, group=1 - Transform, parameterName=Scale, sortPriority=0}, False=AppendVector(A=@n35, B=@n36)) {defaultValue=true, parameterName=UseGlobalScale}
  )
  n38 = ScalarParameter() {group=1 - Transform, parameterName=Offset X, sortPriority=1}
  n39 = ScalarParameter() {group=1 - Transform, parameterName=OffsetY, sortPriority=2}
  NamedRerouteDeclaration(None=Add(A=Add(A=@n37) {constB=0.5}, B=AppendVector(A=@n38, B=@n39))) {name=UVs, nodeColor=<rgba 0 0 0 1>}
  NamedRerouteUsage()
  StaticSwitchParameter(
   True=NamedRerouteUsage()
   False=VectorParameter() {group=3 - Color, parameterName=Color, sortPriority=1, defaultValue=<rgba 1 0.7882 0 1>}.RGB
  ) {defaultValue=true, group=3 - Color, parameterName=Use Color per Channel, sortPriority=0}
 ]
}
